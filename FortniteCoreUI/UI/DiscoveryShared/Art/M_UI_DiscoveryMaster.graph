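# UI master material for Discovery screens, switching statically between a key-art tile and a screen-aligned background.
# Tiles fit key art to widget aspect, add disabled, brighten and Winterfest snow treatments, then shadows and animated hover frame plates.
# Backgrounds blend rotating soft noise with radial and linear gradients, optionally compositing key art through angled falloff and animated triangle masks.
# Opacity comes from a soft edge, angled right bar or constant parameter.
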Material /FortniteCoreUI/UI/DiscoveryShared/Art/M_UI_DiscoveryMaster
graph {
 properties {bUsedWithStaticMesh=false, materialDomain=MD_UI, pixelDepthOffsetMode=PDOM_Legacy, usedShadingModels=Default Lit}
 nodes [
  n1 = Panner(
   Coordinate=NamedRerouteUsage()
   Speed=AppendVector(A=ScalarParameter() {defaultValue=0.02, group=Winterfest, parameterName=SnowSpeed_X}, B=Constant())
  ) {speedY=-0.5}
  n2 = ScalarParameter() {defaultValue=32, group=Winterfest, parameterName=SnowTile}
  n3 = ScalarParameter() {defaultValue=-0.3, group=Winterfest, parameterName=SnowSpeed_Y_A}
  n4 = ScalarParameter() {defaultValue=-0.6, group=Winterfest, parameterName=SnowSpeed_Y_B}
  n5 = Multiply(A=ScalarParameter() {defaultValue=32, group=Winterfest, parameterName=SnowTile}) {constB=0.5}
  n6 = Constant()
  n7 = Sine(None=Multiply(A=@n1, B=AppendVector(A=@n5, B=@n6)))
  n8 = Saturate(None=ComponentMask(None=Ceil(None=@n7)) {r=true})
  n9 = Panner(Coordinate=Add(A=Multiply(A=@n1, B=@n2) {constB=4}), Speed=AppendVector(A=Constant(), B=LinearInterpolate(A=@n3, B=@n4, Alpha=@n8))) {speedY=-0.5}
  n10 = Frac(None=@n9)
  n11 = MaterialFunctionCall!(
   PrimaryTextureCoordinates=@n9
   2ndOrderTextureCoordinates
   RandomSeed=ScalarParameter() {defaultValue=11, group=Winterfest, parameterName=Seed2}
   NoiseDensityBias
   Bypass Noise Influence Crosstalk
  ) {materialFunction=/Game/Effects/Fort_Effects/Materials/_Functions/MF_RandomUVSample}
  n12 = Saturate(None=@n11.NoiseWithRange)
  n13 = MaterialFunctionCall(PrimaryTextureCoordinates=@n9, RandomSeed=ScalarParameter() {defaultValue=34, group=Winterfest, parameterName=Seed3}) {materialFunction=&MF_RandomUVSample}
  n14 = Saturate(None=@n13.NoiseWithRange)
  n15 = DotProduct(A=@n12, B=@n14)
  n16 = Multiply(
   A=Time()
   B=Multiply(A=Add(A=@n15) {constB=-0.5}, B=ScalarParameter() {defaultValue=2.5, group=Winterfest, parameterName=SnowRotationMultiplier_A}) {constB=4}
  )
  n17 = ScalarParameter() {defaultValue=0.25, group=Winterfest, parameterName=Snow_Scale_Low}
  n18 = Sine(None=Multiply(A=Time(), B=ScalarParameter() {defaultValue=0.25, group=Winterfest, parameterName=SineSpeed}) {constB=0.25})
  n19 = LinearInterpolate(A=@n12, B=@n14, Alpha=Add(A=Multiply(A=@n18) {constB=0.5}) {constB=0.5})
  n20 = AppendVector(
   A=@n17
   B=LinearInterpolate(A=@n17, B=ScalarParameter() {defaultValue=0.125, group=Winterfest, parameterName=Snow_Scale_Low_Alt}, Alpha=@n19)
  )
  n21 = ScalarParameter() {defaultValue=1, group=Winterfest, parameterName=Snow_Scale_High}
  n22 = AppendVector(
   A=@n21
   B=LinearInterpolate(A=@n21, B=ScalarParameter() {defaultValue=1.125, group=Winterfest, parameterName=Snow_Scale_High_Alt}, Alpha=@n19)
  )
  n23 = TextureSample(
   UVs=MaterialFunctionCall(UVs=Rotator(Coordinate=@n10, Time=@n16) {speed=1}, Texture Scale=LinearInterpolate(A=@n20, B=@n22, Alpha=@n19)) {materialFunction=&ScaleUVsByCenter}.UVs
   Tex=TextureObject() {samplerType=SAMPLERTYPE_Masks, texture=&/T_UI_Discovery_Snowflake}
  ) {samplerType=SAMPLERTYPE_Masks, texture=&T_UI_Discovery_Snowflake}
  n24 = Saturate(None=@n11.Noise)
  n25 = Multiply(
   A=Multiply(A=LinearInterpolate(A=@n23.R, B=@n23.R, Alpha=@n24), B=ScalarParameter() {defaultValue=1.25, group=Winterfest, parameterName=FlakeSharpen})
   B=@n19
  ) {constA=1}
  n26 = Saturate(None=DotProduct(A=@n24, B=Saturate(None=@n13.Noise)))
  NamedRerouteDeclaration(None=Saturate(None=LinearInterpolate(B=@n25, Alpha=@n26))) {name=WinterfestSnowfallAlpha, nodeColor=<rgba 1 0.2578 0 1>}
  n27 = VectorParameter() {group=Image, parameterName=MissingKeyArtColorA, defaultValue=<rgba 0.0001 0.0004 0.0104 1>}
  n28 = VectorParameter() {group=Image, parameterName=MissingKeyArtColorB, defaultValue=<rgba 0 0.0252 0.2542 1>}
  n29 = ScalarParameter() {group=Triangles, parameterName=TriSS_Min}
  n30 = ScalarParameter() {defaultValue=1, group=Triangles, parameterName=TriSS_Max}
  n31 = ComponentMask(None=Add(A=TextureCoordinate() {coordinateIndex=4}, B=ScalarParameter() {defaultValue=-0.25, group=Image, parameterName=MissingTris_Shift})) {g=true}
  n32 = Constant() {r=0.5}
  n33 = Constant() {r=3}
  n34 = Power(
   Base=Multiply(
    A=MaterialFunctionCall!(X=@n31, Tangent 0=@n32, Tangent 1=@n33) {materialFunction=/Engine/Functions/Engine_MaterialFunctions02/Gradients/SmoothCurve}.Result
    B=ScalarParameter() {defaultValue=4, group=Image, parameterName=MissingTris_Mult}
   )
   Exp=ScalarParameter() {defaultValue=1, group=Image, parameterName=MissingTris_Power}
  )
  n35 = LinearInterpolate(A=NamedRerouteUsage(), Alpha=OneMinus(None=Saturate(None=@n34)))
  n36 = Subtract(A=NamedRerouteUsage(), B=@n35) {constB=0}
  n37 = Multiply(
   A=Saturate(None=SmoothStep(Min=@n29, Max=@n30, Value=@n36))
   B=Power(Base=OneMinus(None=@n35), Exp=ScalarParameter() {defaultValue=6, group=Triangles, parameterName=TriNoise_FalloffPower})
  )
  n38 = NamedRerouteUsage()
  n39 = Saturate(None=Add(A=Saturate(None=@n37), B=@n38))
  NamedRerouteDeclaration(None=StaticSwitchParameter(True=LinearInterpolate(A=@n27.RGB, B=@n28.RGB, Alpha=@n39), False=LinearInterpolate(A=@n27.RGB, B=@n28.RGB, Alpha=@n38)) {group=Image, parameterName=bUseTriangles}) {name=MissingInfo, nodeColor=<rgba 0 0 0 1>}
  n40 = TextureCoordinate()
  n41 = Add(
   A=Rotator(Coordinate=@n40, Time=Constant() {r=0.5}) {speed=6.2832}
   B=ScalarParameter() {defaultValue=-0.025, group=Falloff, parameterName=BottomFalloff_YOffset}
  )
  n42 = ScalarParameter() {defaultValue=-0.0188, group=Falloff, parameterName=SideFalloff_Rotation}
  n43 = ScalarParameter() {defaultValue=-0.0625, group=Falloff, parameterName=SideFalloff_XOffset}
  n44 = Add(A=Rotator(Coordinate=@n40, Time=@n42) {speed=6.2832}, B=@n43)
  n45 = Constant() {r=0.5}
  n46 = Add(A=Rotator(Coordinate=@n40, Time=Add(A=@n42, B=@n45)) {speed=6.2832}, B=@n43)
  n47 = AppendVector(A=AppendVector(A=ComponentMask(None=@n41) {g=true}, B=ComponentMask(None=@n44) {r=true}), B=ComponentMask(None=@n46) {r=true})
  n48 = ScalarParameter() {group=Animators, parameterName=NoImage_Animate, sliderMax=1}
  n49 = Add(
   A=@n47
   B=LinearInterpolate(
    A=ScalarParameter() {defaultValue=-0.0125, group=Falloff Cut, parameterName=TriCut_FadeFalloff_Shift_A}
    B=ScalarParameter() {defaultValue=-0.0125, group=Falloff Cut, parameterName=TriCut_FadeFalloff_Shift_B}
    Alpha=@n48
   )
  )
  n50 = ScalarParameter() {defaultValue=4, group=Falloff Cut, parameterName=TriCut_FadeFalloff_Scale}
  n51 = Saturate(None=Power(Base=Multiply(A=@n49, B=@n50), Exp=ScalarParameter() {defaultValue=1.5, group=Falloff Cut, parameterName=TriCut_FadeFalloff_Power}))
  n52 = ComponentMask(None=@n51) {g=true}
  n53 = ComponentMask(None=@n51) {b=true}
  n54 = StaticSwitchParameter(True=@n52, False=Multiply(A=Multiply(A=@n52, B=@n53), B=ComponentMask(None=@n51) {r=true})) {group=PrimarySwitches, parameterName=UseLeftFalloffOnly}
  NamedRerouteDeclaration(None=OneMinus(None=Saturate(None=@n54))) {name=TriCutoffMask, nodeColor=<rgba 1 0 0.3281 1>}
  NamedRerouteUsage()
  n55 = NamedRerouteUsage()
  n56 = Constant() {r=1}
  n57 = ScalarParameter() {defaultValue=0.95, group=Frame And Shadows, parameterName=AllShadowsBurn}
  n58 = MaterialFunctionCall(Float2=TextureCoordinate() {coordinateIndex=4}) {materialFunction=&BreakOutFloat2Components}
  n59 = NamedRerouteUsage()
  n60 = AppendVector(A=@n58.R, B=Multiply(A=@n58.G, B=@n59) {constB=0.5})
  n61 = AppendVector(
   A=ScalarParameter() {defaultValue=1, group=RadialGrad, parameterName=RadShadow_XOffset}
   B=Divide(A=ScalarParameter() {defaultValue=1, group=RadialGrad, parameterName=RadShadow_YOffset}, B=NamedRerouteUsage())
  )
  n62 = Multiply(A=ScalarParameter() {defaultValue=0.5, group=RadialGrad, parameterName=RadShadow_Radius}, B=@n59)
  n63 = ScalarParameter() {defaultValue=4, group=RadialGrad, parameterName=RadShadow_Density}
  n64 = NamedRerouteDeclaration(None=Saturate(None=MaterialFunctionCall(UVs=@n60, CenterPosition=@n61, Radius=@n62, Density=@n63) {materialFunction=&RadialGradientExponential}.RadialGradientExponential)) {name=RadialShadow, nodeColor=<rgba 0.3594 0 1 1>}
  n65 = ScalarParameter() {group=Animators, parameterName=ImageBrighten_Animate, sliderMax=1}
  n66 = Constant() {r=1}
  MaterialExpressionTextureObjectParameter_0 = TextureObjectParameter() {group=Image, parameterName=ModeKeyArt, texture=/Game/UI/DiscoveryShared/Art/T_UI_Discovery_DefaultEmpty}
  n67 = Custom!(Tex=@MaterialExpressionTextureObjectParameter_0) {
   code="float2 dim;\r\nTextureObject.GetDimensions(dim.x, dim.y);\r\nreturn dim;"
   description=GetDimenstions
   outputType=CMOT_Float2
   inputs [{
     inputName=TextureObject
     input {expression=&self:MaterialExpressionTextureObjectParameter_0}
    }]
  }
  n68 = MaterialFunctionCall(Float2=@n67) {materialFunction=&BreakOutFloat2Components}
  n69 = Divide(A=@n68.R, B=@n68.R)
  n70 = MaterialFunctionCall(Float2=MaterialFunctionCall() {materialFunction=&GetUserInterfaceUV}.Pixel Size) {materialFunction=&BreakOutFloat2Components}
  n71 = NamedRerouteDeclaration(None=Divide(A=@n70.R, B=@n70.R)) {name=PixelDimXDividedByY, nodeColor=<rgba 0.0078 1 0 1>}
  n72 = Constant() {r=1}
  n73 = Divide(A=@n68.G, B=@n68.G)
  n74 = If(A=@n69, B=@n71, "A > B"=AppendVector(A=Divide(A=@n69, B=@n71), B=@n72), "A < B"=AppendVector(A=@n72, B=Multiply(A=@n73, B=@n71)))
  n75 = Constant() {r=1}
  n76 = AppendVector(A=@n69, B=@n75)
  n77 = Divide(A=@n70.G, B=@n70.G)
  n78 = AppendVector(A=Multiply(A=NamedRerouteDeclaration(None=@n77) {name=PixelDimYDividedByX, nodeColor=<rgba 0 0.7422 1 1>}, B=@n69), B=@n75)
  n79 = ScalarParameter() {group=Image, parameterName=ImageScaleAdd_Start}
  n80 = ScalarParameter() {defaultValue=0.25, group=Image, parameterName=ImageScaleAdd_Finish}
  n81 = ScalarParameter() {group=Animators, parameterName=ImageScaleAdd_Animate, sliderMax=1}
  n82 = MaterialFunctionCall(
   UVs=TextureCoordinate() {coordinateIndex=4}
   Texture Scale=Add(A=If(A=@n71, B=@n66, "A > B"=@n74, "A == B"=@n76, "A < B"=@n78), B=LinearInterpolate(A=@n79, B=@n80, Alpha=@n81))
  ) {materialFunction=&ScaleUVsByCenter}
  n83 = Multiply(
   A=TextureSample(UVs=@n82.UVs) {texture=&/T_UI_Discovery_Winterfest_Light}.RGB
   B=Constant3Vector() {constant=<rgba 1 0.7594 0.3698 0>}
  )
  n84 = Multiply(A=Time()) {constB=0.5}
  n85 = Add(A=Multiply(A=Sine(None=@n84)) {constB=0.5}) {constB=0.5}
  n86 = Saturate(None=Multiply(A=NamedRerouteUsage()) {constB=32})
  n87 = TextureSample(UVs=@n82.UVs, Tex=@MaterialExpressionTextureObjectParameter_0) {samplerSource=SSM_Clamp_WorldGroupSettings}
  n88 = Divide(A=@n87.RGB, B=LinearInterpolate(A=OneMinus(None=@n83), B=Constant() {r=1}, Alpha=LinearInterpolate(B=@n85, Alpha=@n86) {constA=1}))
  n89 = LinearInterpolate(
   A=Saturate(None=Multiply(A=@n87.RGB) {constB=32})
   B=VectorParameter() {group=Winterfest, parameterName=SnowColor, defaultValue=<rgba 0.4233 0.4564 0.6105 1>}.RGB
  )
  n90 = NamedRerouteUsage()
  n91 = StaticSwitchParameter(True=Saturate(None=LinearInterpolate(A=@n88, B=@n89, Alpha=@n90)), False=@n87.RGB) {group=PrimarySwitches, parameterName=bIsWinterfest}
  n92 = Desaturation(None=@n91, Fraction=ScalarParameter() {defaultValue=0.75, group=Image, parameterName=Disabled_DesatFactor})
  n93 = VectorParameter() {group=Image, parameterName=Disabled_Colorization, defaultValue=<rgba 0.0003 0.0018 0.0497 1>}
  n94 = LinearInterpolate(
   A=Multiply(A=Add(A=@n92, B=@n93.RGB), B=VectorParameter() {group=Image, parameterName=Disabled_Multiply, defaultValue=<rgba 0.5208 0.7365 1 1>}.RGB)
   B=VectorParameter() {group=Image, parameterName=Disabled_Overlay, defaultValue=<rgba 0.0003 0.0018 0.0497 1>}.RGB
   Alpha=LinearInterpolate(
    A=ScalarParameter() {defaultValue=0.5, group=Image, parameterName=DisabledOverlayAlpha_A}
    B=ScalarParameter() {defaultValue=0.8, group=Image, parameterName=DisabledOverlayAlpha_B}
    Alpha=ScalarParameter() {group=Animators, parameterName=DisabledOverlayAlpha_Animate}
   )
  )
  n95 = ScalarParameter() {group=Animators, parameterName=DesaturateAndDarken_Animate}
  n96 = StaticSwitchParameter(True=Saturate(None=LinearInterpolate(A=@n91, B=@n94, Alpha=@n95)), False=@n91) {group=PrimarySwitches, parameterName=UseImageSaturationAdjustment}
  n97 = Add(
   A=@n96
   B=LinearInterpolate(A=Constant(), B=ScalarParameter() {defaultValue=0.0125, group=Frame And Shadows, parameterName=ImageHover_Add}, Alpha=@n65)
  )
  n98 = OneMinus(None=LinearInterpolate(A=Constant(), B=ScalarParameter() {defaultValue=0.5, group=Frame And Shadows, parameterName=ImageHover_Dodge}, Alpha=@n65))
  n99 = StaticSwitchParameter(True=Saturate(None=Divide(A=@n97, B=@n98)), False=@n96) {group=PrimarySwitches, parameterName=bUseImageBrightening}
  n100 = Add(A=Add(A=@n99, B=LinearInterpolate(A=@n56, B=@n57, Alpha=@n64)), B=Constant() {r=-1})
  n101 = LinearInterpolate(A=Constant() {r=1}, B=ScalarParameter() {defaultValue=0.2, group=Frame And Shadows, parameterName=AllShadowsMult}, Alpha=@n64)
  n102 = Saturate(None=StaticSwitchParameter(True=Multiply(A=@n100, B=@n101), False=@n99) {group=PrimarySwitches, parameterName=bUseCornerShadow})
  n103 = Saturate(None=LinearInterpolate(
    A=LinearInterpolate(A=@n55, B=@n102, Alpha=@n87.A)
    B=@n55
    Alpha=ScalarParameter() {group=Animators, parameterName=NoImage_Animate, sliderMax=1}
   ))
  n104 = StaticSwitchParameter(
   True=Multiply(A=NamedRerouteUsage(), B=ScalarParameter() {defaultValue=0.25, group=RadialGrad, parameterName=RadShadow_BlueAmount})
   False=Constant()
  ) {group=PrimarySwitches, parameterName=bUseCornerShadow}
  n105 = MaterialFunctionCall(Float2=MaterialFunctionCall() {materialFunction=&ScreenResolution}.Visible Resolution) {materialFunction=&BreakOutFloat2Components}
  n106 = Divide(A=@n105.R, B=Max(A=@n105.G))
  n107 = Constant() {r=1}
  n108 = NamedRerouteDeclaration(None=Multiply(
    A=AppendVector(A=@n106, B=@n107)
    B=MaterialFunctionCall() {materialFunction=&ScreenAlignedUVs}."X 100%, Y 100%"
   )) {name=Screen-Aligned+SquaredUVs, nodeColor=<rgba 1 0.8883 0 1>}
  n109 = Rotator(Coordinate=@n108, Time=ScalarParameter() {defaultValue=0.125, group=Triangles, parameterName=TriNoise_Rotation}) {speed=6.2832}
  n110 = ScalarParameter() {defaultValue=24, group=Triangles, parameterName=TriNoise_Tile}
  n111 = Time()
  n112 = AppendVector(
   A=Multiply(A=@n111, B=ScalarParameter() {defaultValue=-0.0625, group=Triangles, parameterName=TriNoise_Pan_X}) {constB=0.25}
   B=Multiply(A=@n111, B=ScalarParameter() {defaultValue=0.0312, group=Triangles, parameterName=TriNoise_Pan_Y}) {constB=0.25}
  )
  n113 = TextureObjectParameter() {group=Triangles, parameterName=Triangles, texture=/Game/Global/Textures/Default/Masks_and_Gradients/T-BattlePass-Tile-Triangles}
  n114 = TextureSample(UVs=Add(A=Multiply(A=@n109, B=@n110) {constB=3}, B=@n112), Tex=@n113)
  n115 = LinearInterpolate(A=@n114.RGB, B=OneMinus(None=@n114.RGB), Alpha=ScalarParameter() {group=Triangles, parameterName=InvertTriangles})
  n116 = Panner(
   Coordinate=Multiply(A=@n108, B=ScalarParameter() {defaultValue=3, group=Triangles, parameterName=TriNoise_MissingTris_Tile})
   Speed=AppendVector(
    A=ScalarParameter() {defaultValue=0.05, group=Triangles, parameterName=TriNoise_SppedX}
    B=ScalarParameter() {defaultValue=0.12, group=Triangles, parameterName=TriNoise_SpeedY}
   )
  )
  n117 = TextureObject() {samplerType=SAMPLERTYPE_Grayscale, texture=&LowResBlurredNoise}
  n118 = Multiply(
   A=ComponentMask(None=TextureSample(UVs=@n116, Tex=@n117) {samplerType=SAMPLERTYPE_Grayscale, texture=&LowResBlurredNoise}.RGB) {r=true}
   B=ScalarParameter() {defaultValue=2, group=Triangles, parameterName=TriNoise_Mult}
  )
  n119 = ScalarParameter() {defaultValue=3, group=Triangles, parameterName=TriNoise_Power}
  n120 = LinearInterpolate(
   A=ScalarParameter() {group=Triangles, parameterName=TriNoise_InversePower_Low}
   B=ScalarParameter() {defaultValue=1, group=Triangles, parameterName=TriNoise_InversePower_High}
   Alpha=Saturate(None=Power(Base=@n118, Exp=@n119))
  )
  n121 = Multiply(A=@n110) {constB=0.25}
  n122 = TextureObjectParameter() {
   group=Triangles
   parameterName=MissingTriangles
   samplerType=SAMPLERTYPE_Masks
   texture=&/T_UI_Discovery_TriangleFalloff
  }
  n123 = TextureSample(UVs=Add(A=Multiply(A=@n109, B=@n121), B=Multiply(A=@n112) {constB=0.25}), Tex=@n122)
  n124 = Multiply(A=@n123.G, B=ScalarParameter() {defaultValue=1, group=Triangles, parameterName=TriNoise_MultLayer_A}) {constB=0.25}
  n125 = ScalarParameter() {defaultValue=0.5, group=Triangles, parameterName=TriNoise_MultLayer_B}
  n126 = Saturate(None=Add(A=Add(A=@n123.R, B=@n124), B=Multiply(A=@n123.B, B=@n125)))
  n127 = ComponentMask(None=@n41) {g=true}
  n128 = ComponentMask(None=@n44) {r=true}
  n129 = Add(
   A=AppendVector(A=AppendVector(A=@n127, B=@n128), B=ComponentMask(None=@n46) {r=true})
   B=LinearInterpolate(
    A=ScalarParameter() {defaultValue=-0.0125, group=Falloff, parameterName=A_TriFalloff_Shift}
    B=ScalarParameter() {defaultValue=-0.0125, group=Falloff, parameterName=B_TriFalloff_Shift}
    Alpha=@n48
   )
  )
  n130 = LinearInterpolate(
   A=ScalarParameter() {defaultValue=3.5, group=Falloff, parameterName=A_TriFalloff_Scale}
   B=ScalarParameter() {defaultValue=3.5, group=Falloff, parameterName=B_TriFalloff_Scale}
   Alpha=@n48
  )
  n131 = ScalarParameter() {defaultValue=2.5, group=Falloff, parameterName=A_TriFalloff_Power}
  n132 = ScalarParameter() {defaultValue=2.5, group=Falloff, parameterName=B_TriFalloff_Power}
  n133 = Saturate(None=Power(Base=Multiply(A=@n129, B=@n130), Exp=LinearInterpolate(A=@n131, B=@n132, Alpha=@n48)))
  n134 = ComponentMask(None=@n133) {g=true}
  n135 = Multiply(A=ComponentMask(None=@n133) {r=true}, B=Multiply(A=@n134, B=ComponentMask(None=@n133) {b=true}))
  n136 = LinearInterpolate(
   A=NamedRerouteDeclaration(None=LinearInterpolate(B=@n120, Alpha=@n126)) {name=TriangleNoise, nodeColor=<rgba 1 0.0493 0 1>}
   Alpha=Saturate(None=StaticSwitchParameter(True=@n134, False=@n135) {group=PrimarySwitches, parameterName=UseLeftFalloffOnly})
  )
  n137 = SmoothStep(
   Min=ScalarParameter() {group=Triangles, parameterName=TriSS_Min}
   Max=ScalarParameter() {defaultValue=1, group=Triangles, parameterName=TriSS_Max}
   Value=Subtract(A=NamedRerouteDeclaration(None=@n115) {name=TrianglePattern, nodeColor=<rgba 0.4062 1 0 1>}, B=@n136) {constB=0}
  )
  n138 = OneMinus(None=@n136)
  n139 = ScalarParameter() {defaultValue=6, group=Triangles, parameterName=TriNoise_FalloffPower}
  n140 = Add(A=@n47, B=ScalarParameter() {defaultValue=-0.0125, group=Falloff, parameterName=FadeFalloff_Shift})
  n141 = LinearInterpolate(
   A=ScalarParameter() {defaultValue=4.5, group=Falloff, parameterName=A_FadeFalloff_Scale}
   B=ScalarParameter() {defaultValue=4.5, group=Falloff, parameterName=B_FadeFalloff_Scale}
   Alpha=@n48
  )
  n142 = ScalarParameter() {defaultValue=1.5, group=Falloff, parameterName=A_FadeFalloff_Power}
  n143 = ScalarParameter() {defaultValue=1.5, group=Falloff, parameterName=B_FadeFalloff_Power}
  n144 = Saturate(None=Power(Base=Multiply(A=@n140, B=@n141), Exp=LinearInterpolate(A=@n142, B=@n143, Alpha=@n48)))
  n145 = ComponentMask(None=@n144) {g=true}
  n146 = Multiply(A=Multiply(A=@n145, B=ComponentMask(None=@n144) {b=true}), B=ComponentMask(None=@n144) {r=true})
  n147 = OneMinus(None=Saturate(None=StaticSwitchParameter(True=@n145, False=@n146) {group=PrimarySwitches, parameterName=UseLeftFalloffOnly}))
  n148 = Add(A=@n47, B=ScalarParameter() {defaultValue=-0.0125, group=Falloff Cut, parameterName=Cut_FadeFalloff_Shift})
  n149 = ScalarParameter() {defaultValue=4, group=Falloff Cut, parameterName=Cut_FadeFalloff_Scale}
  n150 = Saturate(None=Power(Base=Multiply(A=@n148, B=@n149), Exp=ScalarParameter() {defaultValue=1.5, group=Falloff Cut, parameterName=Cut_FadeFalloff_Power}))
  n151 = ComponentMask(None=@n150) {g=true}
  n152 = Multiply(A=@n151, B=ComponentMask(None=@n150) {b=true})
  n153 = ComponentMask(None=@n150) {r=true}
  n154 = Saturate(None=StaticSwitchParameter(True=@n151, False=Multiply(A=@n152, B=@n153)) {group=PrimarySwitches, parameterName=UseLeftFalloffOnly})
  n155 = Add(
   A=Multiply(A=Saturate(None=@n137), B=Power(Base=@n138, Exp=@n139))
   B=NamedRerouteDeclaration(None=LinearInterpolate(A=@n147, Alpha=@n154) {constB=0}) {name=AngleMask, nodeColor=<rgba 0 1 0.3252 1>}
  )
  n156 = Saturate(None=OneMinus(None=Add(A=@n104, B=@n155)))
  n157 = VectorParameter() {group=RadialGrad, parameterName=PrimaryRadial_Outer, defaultValue=<rgba 0 0 0.0331 1>}
  n158 = MaterialFunctionCall(Float2=MaterialFunctionCall() {materialFunction=&ScreenResolution}.Visible Resolution) {materialFunction=&BreakOutFloat2Components}
  n159 = Max(A=@n158.G)
  n160 = MaterialFunctionCall() {materialFunction=&ScreenAlignedUVs}
  n161 = Multiply(A=AppendVector(A=Divide(A=@n158.R, B=@n159), B=Constant() {r=1}), B=@n160."X 100%, Y 100%")
  n162 = Rotator(Coordinate=@n161, Time=Multiply(A=Time(), B=ScalarParameter() {defaultValue=-0.4, group=Noise, parameterName=NoiseB_RotationTime}))
  n163 = ScalarParameter() {defaultValue=0.18, group=Noise, parameterName=NoiseB_Tile}
  n164 = ScalarParameter() {defaultValue=-0.026, group=Noise, parameterName=NoiseB_Speed_X}
  n165 = ScalarParameter() {defaultValue=-0.044, group=Noise, parameterName=NoiseB_Speed_Y}
  n166 = TextureObject() {samplerType=SAMPLERTYPE_Grayscale, texture=&LowResBlurredNoise}
  n167 = TextureSample(UVs=Panner(Coordinate=Multiply(A=@n162, B=@n163), Speed=AppendVector(A=@n164, B=@n165)), Tex=@n166) {samplerType=SAMPLERTYPE_Grayscale, texture=&LowResBlurredNoise}
  n168 = Rotator(Coordinate=@n161, Time=Multiply(A=Time(), B=ScalarParameter() {defaultValue=0.3, group=Noise, parameterName=NoiseA_RotationTime}))
  n169 = ScalarParameter() {defaultValue=0.42, group=Noise, parameterName=NoiseA_Tile}
  n170 = ScalarParameter() {defaultValue=0.032, group=Noise, parameterName=NoiseA_Speed_X}
  n171 = ScalarParameter() {defaultValue=0.018, group=Noise, parameterName=NoiseA_Speed_Y}
  n172 = TextureSample(UVs=Panner(Coordinate=Multiply(A=@n168, B=@n169), Speed=AppendVector(A=@n170, B=@n171)), Tex=@n166) {samplerType=SAMPLERTYPE_Grayscale, texture=&LowResBlurredNoise}
  n173 = Power(
   Base=Multiply(A=DotProduct(A=@n167.RGB, B=@n172.RGB), B=ScalarParameter() {defaultValue=1.0625, group=Noise, parameterName=NoiseMult})
   Exp=ScalarParameter() {defaultValue=3, group=Noise, parameterName=NoisePower}
  )
  n174 = LinearInterpolate(
   A=VectorParameter() {group=RadialGrad, parameterName=NoiseColorA, defaultValue=<rgba 0.007 0.011 0.1329 1>}.RGB
   B=VectorParameter() {group=RadialGrad, parameterName=NoiseColorB, defaultValue=<rgba 0.0037 0.2122 0.8879 1>}.RGB
   Alpha=NamedRerouteDeclaration(None=Saturate(None=@n173)) {name=BackgroundSoftNoise, nodeColor=<rgba 1 0 0.2578 1>}
  )
  n175 = Divide(A=@n158.R, B=@n158.G)
  n176 = Multiply(A=AppendVector(A=@n175, B=Constant() {r=1}), B=@n160."X 100%, Y 100%")
  n177 = Add(
   A=AppendVector(A=Multiply(A=@n175) {constB=0.5}, B=Constant() {r=0.5})
   B=AppendVector(
    A=ScalarParameter() {defaultValue=0.14, group=RadialGrad, parameterName=RGrad_XOffset}
    B=ScalarParameter() {defaultValue=0.8, group=RadialGrad, parameterName=RGrad_YOffset}
   )
  )
  n178 = ScalarParameter() {defaultValue=1.4, group=RadialGrad, parameterName=RGrad_Radius}
  n179 = ScalarParameter() {defaultValue=3.5, group=RadialGrad, parameterName=RGrad_Density}
  n180 = LinearInterpolate(
   A=ScalarParameter() {defaultValue=0.05, group=RadialGrad, parameterName=RGrad_DarkenAmount}
   Alpha=Saturate(None=MaterialFunctionCall(UVs=@n176, CenterPosition=@n177, Radius=@n178, Density=@n179) {materialFunction=&RadialGradientExponential}.RadialGradientExponential)
  )
  n181 = Multiply(A=@n176, B=ScalarParameter() {defaultValue=0.5, group=LinearGrad, parameterName=LGrad_Scale})
  n182 = ScalarParameter() {defaultValue=0.2, group=LinearGrad, parameterName=LGrad_Rotation}
  n183 = ComponentMask(None=Add(A=Rotator(Coordinate=@n181, Time=@n182), B=ScalarParameter() {defaultValue=0.4375, group=LinearGrad, parameterName=LGrad_YOffset})) {g=true}
  n184 = ScalarParameter() {defaultValue=1.5, group=LinearGrad, parameterName=LGrad_Power}
  n185 = LinearInterpolate(B=ScalarParameter() {defaultValue=0.3, group=LinearGrad, parameterName=LGrad_Amount}, Alpha=Saturate(None=Power(Base=@n183, Exp=@n184)))
  n186 = NamedRerouteDeclaration(None=Divide(A=LinearInterpolate(A=@n157.RGB, B=@n174, Alpha=@n180), B=OneMinus(None=@n185))) {name=BackgroundRGB, nodeColor=<rgba 0 1 0.5218 1>}
  n187 = TextureCoordinate() {coordinateIndex=4}
  n188 = Constant() {r=1}
  n189 = NamedRerouteUsage()
  n190 = Rotator(
   Coordinate=Multiply(A=@n187, B=AppendVector(A=@n188, B=@n189))
   Time=ScalarParameter() {defaultValue=0.125, group=Frame And Shadows, parameterName=HeartShadow_Rotation}
  ) {speed=6.2832}
  n191 = LinearInterpolate(
   A=LinearInterpolate(
    A=Constant() {r=1}
    B=ScalarParameter() {defaultValue=0.125, group=Frame And Shadows, parameterName=HeartShadow_Start_Fav}
    Alpha=ScalarParameter() {defaultValue=1, group=Frame And Shadows, parameterName=bIsFavorite, sliderMax=1}
   )
   B=ScalarParameter() {defaultValue=0.125, group=Frame And Shadows, parameterName=HeartShadow_Finish}
   Alpha=ScalarParameter() {group=Animators, parameterName=TextHover_Animate, sliderMax=1}
  )
  n192 = NamedRerouteDeclaration(None=ComponentMask(None=Add(A=@n190, B=@n191) {constB=0}) {g=true}) {name=HeartShadow, nodeColor=<rgba 0 0.4229 1 1>}
  n193 = Constant() {r=1}
  n194 = ScalarParameter() {defaultValue=0.125, group=Frame And Shadows, parameterName=DetailsShadow_Finish}
  n195 = ScalarParameter() {group=Animators, parameterName=TextHover_Animate, sliderMax=1}
  n196 = ComponentMask(None=Add(A=OneMinus(None=@n190), B=LinearInterpolate(A=@n193, B=@n194, Alpha=@n195)) {constB=0}) {g=true}
  n197 = ScalarParameter() {defaultValue=0.9125, group=Frame And Shadows, parameterName=TextShadow_Shift_Start}
  n198 = ScalarParameter() {defaultValue=1.25, group=Frame And Shadows, parameterName=TextShadow_Shift_Finish}
  n199 = ScalarParameter() {group=Animators, parameterName=TextHover_Animate, sliderMax=1}
  n200 = Add(
   A=Rotator(Coordinate=@n187, Time=ScalarParameter() {defaultValue=-0.51, group=Frame And Shadows, parameterName=TextShadow_Rotation}) {speed=6.2832}
   B=Add(A=LinearInterpolate(A=@n197, B=@n198, Alpha=@n199)) {constB=-1}
  )
  n201 = Multiply(
   A=AppendVector(A=AppendVector(A=@n192, B=@n196), B=ComponentMask(None=@n200) {g=true})
   B=ScalarParameter() {defaultValue=4, group=Frame And Shadows, parameterName=AllShadows_Scale}
  )
  n202 = ScalarParameter() {defaultValue=2.5, group=Frame And Shadows, parameterName=AllShadows_Power}
  n203 = Saturate(None=OneMinus(None=Power(Base=@n201, Exp=@n202)))
  n204 = ComponentMask(None=@n203) {b=true}
  n205 = Add(
   A=LinearInterpolate(
    B=ComponentMask(None=@n203) {r=true}
    Alpha=ScalarParameter() {defaultValue=1, group=Frame And Shadows, parameterName=bHasFavoriteButton, sliderMax=1}
   )
   B=LinearInterpolate(
    B=ComponentMask(None=@n203) {g=true}
    Alpha=ScalarParameter() {defaultValue=1, group=Frame And Shadows, parameterName=bHasDetailsButton, sliderMax=1}
   )
  )
  n206 = LinearInterpolate(A=Saturate(None=LinearInterpolate(A=@n204, Alpha=@n205)), B=Constant(), Alpha=ScalarParameter() {group=Image, parameterName=MissingKeyArt})
  n207 = Add(
   A=@n99
   B=LinearInterpolate(A=Constant() {r=1}, B=ScalarParameter() {defaultValue=0.95, group=Frame And Shadows, parameterName=AllShadowsBurn}, Alpha=@n206)
  )
  n208 = Constant() {r=-1}
  n209 = Constant() {r=1}
  n210 = ScalarParameter() {defaultValue=0.2, group=Frame And Shadows, parameterName=AllShadowsMult}
  n211 = NamedRerouteUsage()
  n212 = LinearInterpolate(
   A=Multiply(A=Add(A=@n207, B=@n208), B=LinearInterpolate(A=@n209, B=@n210, Alpha=@n206))
   B=@n211
   Alpha=ScalarParameter() {group=Animators, parameterName=NoImage_Animate, sliderMax=1}
  )
  n213 = VectorParameter() {group=Frame And Shadows, parameterName=YellowHoverColor_Disabled, defaultValue=<rgba 0.7835 0.9301 1 1>}
  n214 = TextureCoordinate() {coordinateIndex=4}
  n215 = ScalarParameter() {defaultValue=-0.003, group=Frame And Shadows, parameterName=YellowHover_Rotation}
  n216 = ScalarParameter() {group=Animators, parameterName=TextHover_Animate, sliderMax=1}
  n217 = LinearInterpolate(
   A=ScalarParameter() {defaultValue=-0.125, group=Frame And Shadows, parameterName=YellowHover_PlateStart}
   B=ScalarParameter() {defaultValue=0.05, group=Frame And Shadows, parameterName=YellowHover_PlateFinish}
   Alpha=@n216
  )
  n218 = ScalarParameter() {defaultValue=256, group=Frame And Shadows, parameterName=FrameSharpness}
  n219 = Multiply(A=Add(A=Rotator(Coordinate=@n214, Time=@n215) {speed=6.2832}, B=Add(A=@n217) {constB=-1}), B=@n218)
  n220 = VectorParameter() {group=Frame And Shadows, parameterName=FrameBorderColor_Mid, defaultValue=<rgba 0 0.0908 1 1>}
  n221 = VectorParameter() {group=Frame And Shadows, parameterName=FrameBorderColor_Bright, defaultValue=<rgba 0.0103 0.3467 1 1>}
  n222 = NamedRerouteUsage()
  n223 = VectorParameter() {group=Frame And Shadows, parameterName=FrameBorderColor_DisabledHover, defaultValue=<rgba 0.402 0.6172 0.8879 1>}
  n224 = LinearInterpolate(
   A=StaticSwitchParameter(True=LinearInterpolate(A=@n220.RGB, B=@n221.RGB, Alpha=@n222), False=@n220.RGB) {defaultValue=true, group=PrimarySwitches, parameterName=UseSoftNoise}
   B=LinearInterpolate(A=Constant() {r=1}, B=@n223.RGB, Alpha=ScalarParameter() {group=Animators, parameterName=bIsDisabledFrame_Animate, sliderMax=1})
   Alpha=ScalarParameter() {group=Animators, parameterName=TextHover_Animate, sliderMax=1}
  )
  n225 = LinearInterpolate(
   A=@n224
   B=LinearInterpolate(
    A=VectorParameter() {group=Frame And Shadows, parameterName=YellowHoverColor, defaultValue=<rgba 0.9301 1 0.0103 1>}.RGB
    B=@n213.RGB
    Alpha=ScalarParameter() {group=Animators, parameterName=bIsDisabledFrame_Animate, sliderMax=1}
   )
   Alpha=Saturate(None=ComponentMask(None=@n219) {g=true})
  )
  n226 = TextureCoordinate()
  n227 = Add(
   A=Abs(None=Add(A=@n226) {constB=-0.5})
   B=Add(A=ScalarParameter() {defaultValue=0.0312, group=Frame And Shadows, parameterName=FramingEdge}) {constB=-0.5}
  )
  n228 = MaterialFunctionCall(Float2=Saturate(None=Multiply(A=@n227, B=@n218))) {materialFunction=&BreakOutFloat2Components}
  n229 = Saturate(None=Add(A=@n228.R, B=@n228.R))
  n230 = Add(A=@n226, B=ScalarParameter() {defaultValue=-0.1525, group=Frame And Shadows, parameterName=FramingEdgeBanner}) {constB=-0.1}
  n231 = ComponentMask(None=Saturate(None=Multiply(A=@n230, B=@n218))) {g=true}
  n232 = LinearInterpolate(
   A=@n229
   B=LinearInterpolate(A=@n229, Alpha=OneMinus(None=@n231))
   Alpha=ScalarParameter() {group=Animators, parameterName=HasTopBanner, sliderMax=1}
  )
  n233 = LinearInterpolate(
   A=ScalarParameter() {defaultValue=-0.03, group=Frame And Shadows, parameterName=TextHover_RotationStart}
   B=ScalarParameter() {defaultValue=0.01, group=Frame And Shadows, parameterName=TextHover_RotationFinish}
   Alpha=ScalarParameter() {group=Animators, parameterName=TextHover_Animate, sliderMax=1}
  )
  n234 = ScalarParameter() {defaultValue=0.275, group=Frame And Shadows, parameterName=TextSelect_RotationFinish}
  n235 = ScalarParameter() {group=Animators, parameterName=TextSelect_Animate, sliderMax=1}
  n236 = LinearInterpolate(
   A=ScalarParameter() {defaultValue=-0.125, group=Frame And Shadows, parameterName=TextHover_PlateStart}
   B=ScalarParameter() {defaultValue=0.275, group=Frame And Shadows, parameterName=TextHover_PlateFinish}
   Alpha=@n216
  )
  n237 = ScalarParameter() {defaultValue=0.275, group=Frame And Shadows, parameterName=TextSelect_PlateFinish}
  n238 = ScalarParameter() {group=Animators, parameterName=TextSelect_Animate, sliderMax=1}
  n239 = Add(
   A=Rotator(Coordinate=@n214, Time=LinearInterpolate(A=@n233, B=@n234, Alpha=@n235)) {speed=6.2832}
   B=Add(A=LinearInterpolate(A=@n236, B=@n237, Alpha=@n238)) {constB=-1}
  )
  n240 = Saturate(None=ComponentMask(None=Multiply(A=@n239, B=@n218)) {g=true})
  StaticSwitchParameter(
   True=LinearInterpolate(
    A=StaticSwitchParameter(True=@n212, False=@n211) {defaultValue=true, group=PrimarySwitches, parameterName=UseKeyArt}
    B=StaticSwitchParameter(True=@n225, False=@n224) {defaultValue=true, group=PrimarySwitches, parameterName=HasYellowStripe}
    Alpha=LinearInterpolate(A=@n232, Alpha=@n240)
   )
   False=StaticSwitchParameter(True=@n186, False=LinearInterpolate(A=@n186, B=@n103, Alpha=@n156)) {defaultValue=true, group=PrimarySwitches, parameterName=bIsBackground}
  ) {group=PrimarySwitches, parameterName=bIsTile}
  n241 = Abs(None=ConstantBiasScale(None=TextureCoordinate()) {bias=-0.5, scale=2})
  n242 = ScalarParameter() {defaultValue=0.98, group=SoftEdge, parameterName=SoftEdge_Size}
  n243 = Saturate(None=Multiply(A=Subtract(A=@n241, B=@n242) {constB=0.9}, B=ScalarParameter() {defaultValue=64, group=SoftEdge, parameterName=SoftEdge_Falloff}))
  n244 = Rotator(
   Coordinate=TextureCoordinate() {coordinateIndex=4}
   Time=LinearInterpolate(
    A=ScalarParameter() {defaultValue=0.48, group=Right Bar, parameterName=Right-Angle}
    B=ScalarParameter() {defaultValue=0.49, group=Right Bar, parameterName=RIght-Angle-hover}
    Alpha=ScalarParameter() {group=Right Bar, parameterName=hover, sliderMax=1}
   )
  ) {speed=6.2832}
  n245 = ScalarParameter() {defaultValue=0.075, group=Right Bar, parameterName=Right-RangeOffset, sliderMax=1}
  n246 = ScalarParameter() {defaultValue=0.04, group=Right Bar, parameterName=Right-RangeOffset-hover, sliderMax=1}
  n247 = ScalarParameter() {group=Animators, parameterName=TextHover_Animate, sliderMax=1}
  n248 = Multiply(
   A=Subtract(A=ComponentMask(None=@n244) {r=true}, B=LinearInterpolate(A=@n245, B=@n246, Alpha=@n247))
   B=ScalarParameter() {defaultValue=128, group=Right Bar, parameterName=Strength}
  )
  StaticSwitchParameter(
   True=OneMinus(None=MaterialFunctionCall(Vector 2=@n243) {materialFunction=/Engine/Functions/Engine_MaterialFunctions02/Utility/VectorLength}.V2 Length)
   False=StaticSwitchParameter(True=Saturate(None=@n248), False=ScalarParameter() {defaultValue=1, parameterName=DefaultOpacity}) {group=PrimarySwitches, parameterName=bIsUseRightBar}
  ) {group=PrimarySwitches, parameterName=bUseSoftEdge}
 ]
}
