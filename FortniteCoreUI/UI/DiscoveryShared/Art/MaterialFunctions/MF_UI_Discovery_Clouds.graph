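# Material function for Discovery UI that generates animated, cloud-like grayscale noise.
# It samples a noise texture, blurred low-res by default, through two rotating, panning UV layers, the second at larger tiling and counter-rotating slightly faster.
# A dot product of both samples forms the moving noise, which is scaled by intensity, raised to a power and saturated.
# UVs, texture, scale, speeds, intensity and power are overridable inputs.
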MaterialFunction /FortniteCoreUI/UI/DiscoveryShared/Art/MaterialFunctions/MF_UI_Discovery_Clouds
graph {nodes [
  n1 = Multiply(
   A=NamedRerouteUsage()
   B=FunctionInput(Preview=Constant() {r=1.0625}) {bUsePreviewValueAsDefault=true, inputName=Noise Intensity, inputType=FunctionInput_Scalar, sortPriority=6}
  )
  n2 = FunctionInput(Preview=Constant() {r=3}) {bUsePreviewValueAsDefault=true, inputName=Noise Power, inputType=FunctionInput_Scalar, sortPriority=7}
  FunctionOutput(None=Saturate(None=Power(Base=@n1, Exp=@n2)))
  NamedRerouteDeclaration(None=FunctionInput(Preview=MaterialFunctionCall() {materialFunction=&GetUserInterfaceUV}.Normalized UV) {bUsePreviewValueAsDefault=true, inputName=UV, inputType=FunctionInput_Vector2}) {name=UV, nodeColor=<rgba 0 0 0 1>}
  NamedRerouteDeclaration(None=FunctionInput(Preview=TextureObject() {samplerType=SAMPLERTYPE_Grayscale, texture=&LowResBlurredNoise}) {bUsePreviewValueAsDefault=true, inputName=Noise Texture, inputType=FunctionInput_Texture2D, sortPriority=1}) {name=Noise Texture, nodeColor=<rgba 0 0 0 1>}
  n3 = FunctionInput(Preview=Constant() {r=0.18}) {bUsePreviewValueAsDefault=true, inputName=Noise Scale, inputType=FunctionInput_Scalar, sortPriority=3}
  n4 = NamedRerouteUsage()
  n5 = Multiply(
   A=Time()
   B=FunctionInput(Preview=Constant() {r=-0.3}) {bUsePreviewValueAsDefault=true, inputName=Rotation Speed, inputType=FunctionInput_Scalar, sortPriority=2}
  )
  n6 = Constant() {r=-0.026}
  n7 = Constant() {r=-0.044}
  n8 = Panner(
   Coordinate=Multiply(A=@n3, B=Rotator(Coordinate=@n4, Time=@n5))
   Speed=FunctionInput(Preview=AppendVector(A=@n6, B=@n7)) {bUsePreviewValueAsDefault=true, inputName=Speed A, inputType=FunctionInput_Vector2, sortPriority=4}
  )
  n9 = NamedRerouteUsage()
  n10 = NamedRerouteUsage()
  n11 = Multiply(A=@n5) {constB=-1.1}
  n12 = Constant() {r=0.032}
  n13 = Constant() {r=0.018}
  n14 = Panner(
   Coordinate=Multiply(A=Multiply(A=@n3) {constB=2.4}, B=Rotator(Coordinate=@n10, Time=@n11))
   Speed=FunctionInput(Preview=AppendVector(A=@n12, B=@n13)) {bUsePreviewValueAsDefault=true, inputName=Speed B, inputType=FunctionInput_Vector2, sortPriority=5}
  )
  NamedRerouteDeclaration(None=DotProduct(
    A=TextureSample(UVs=@n8, Tex=@n9) {samplerType=SAMPLERTYPE_Grayscale, texture=&LowResBlurredNoise}.RGB
    B=TextureSample(UVs=@n14, Tex=@n9) {samplerType=SAMPLERTYPE_Grayscale, texture=&LowResBlurredNoise}.RGB
   )) {name=Moving Noise, nodeColor=<rgba 0 0 0 1>}
 ]}
