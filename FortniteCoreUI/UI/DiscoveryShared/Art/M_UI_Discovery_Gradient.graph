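# UI-domain translucent material for a Discovery gradient background, blending a dark-blue background color toward a brighter blue accent color.
# The Discovery clouds function supplies the blend alpha with a slight negative rotation, and the result is declared as the Outside Background reroute.
# Named reroutes also hold normalized UI UVs and fitted screen-space UVs.
# A separate mask multiplies eased UV components, with no visible consumer.
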Material /FortniteCoreUI/UI/DiscoveryShared/Art/M_UI_Discovery_Gradient
graph {
 properties {
  bUsedWithStaticMesh=false
  blendMode=BLEND_Translucent
  materialDomain=MD_UI
  pixelDepthOffsetMode=PDOM_Legacy
  usedShadingModels=Default Lit
 }
 nodes [
  n1 = NamedRerouteUsage()
  n2 = MaterialFunctionCall() {materialFunction=&MF_UI_UVScaling}
  Multiply(
   A=MaterialFunctionCall("Value (0-1)"=ComponentMask(None=@n1) {g=true}) {materialFunction=&MF_UI_EaseCurves}.EaseInOutSine
   B=MaterialFunctionCall("Value (0-1)"=ComponentMask(None=@n2.ScaleX) {r=true}) {materialFunction=&MF_UI_EaseCurves}.EaseOutQuint
  )
  NamedRerouteDeclaration(None=MaterialFunctionCall() {materialFunction=&GetUserInterfaceUV}.Normalized UV) {name=Normalized UV, nodeColor=<rgba 0 0 0 1>}
  NamedRerouteDeclaration(None=MaterialFunctionCall!(ScreenUV, TextureSize, ScalePivotPoint) {materialFunction=/FortUILibrary/MaterialFunctions/Basics/MF_UI_ScreenSpaceUVScaling}.Scaled to Fit) {name=Screenspace UV, nodeColor=<rgba 0 0 0 1>}
  n3 = VectorParameter() {group=0 - Global, parameterName=Background Color, sortPriority=1, defaultValue=<rgba 0.0037 0.0137 0.117 1>}
  n4 = VectorParameter() {group=0 - Global, parameterName=Accent Color, sortPriority=2, defaultValue=<rgba 0.007 0.2086 0.8796 1>}
  n5 = NamedRerouteUsage()
  n6 = Constant() {r=-0.2}
  NamedRerouteDeclaration(None=LinearInterpolate(
    A=@n3.RGB
    B=@n4.RGB
    Alpha=MaterialFunctionCall!(UV=@n5, Noise Texture, Rotation Speed=@n6, Noise Scale, Speed A, Speed B, Noise Intensity, Noise Power) {materialFunction=/FortniteCoreUI/UI/DiscoveryShared/Art/MaterialFunctions/MF_UI_Discovery_Clouds}.Result
   )) {name=Outside Background, nodeColor=<rgba 0 0 0 1>}
  TextureCoordinate()
 ]
}
