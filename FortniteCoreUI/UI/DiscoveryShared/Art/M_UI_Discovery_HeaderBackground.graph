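# Translucent UI material for the Discovery header background, showing a thumbnail over an animated backdrop.
# Cloud noise blends background and accent colors, with a tile-backing triangle pattern layered on through a radial gradient.
# The zoomable, pivot-adjustable thumbnail scales to fill, optionally using a manual size so images and video match.
# Switchable vertical, horizontal and triangle vignette masks blend it into the backdrop, and an Inverse Order switch mirrors the layout for localization.
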Material /FortniteCoreUI/UI/DiscoveryShared/Art/M_UI_Discovery_HeaderBackground
graph {
 properties {
  bUsedWithStaticMesh=false
  blendMode=BLEND_Translucent
  materialDomain=MD_UI
  pixelDepthOffsetMode=PDOM_Legacy
  usedShadingModels=Default Lit
 }
 nodes [
  n1 = TextureSample(UVs=NamedRerouteUsage(), Tex=NamedRerouteUsage())
  MaterialFunctionCall!(UVs=NamedRerouteUsage(), Alpha=@n1.A) {desc=Used to crop the edge of the texture, materialFunction=/Game/UI/Foundation/Items/ClampAlphaToTransparentBorder}
  n2 = TextureObjectParameter() {group=1 - Thumbnail, parameterName=Thumbnail, sortPriority=1, texture=/Game/UI/DiscoveryShared/Art/T_UI_Discovery_DefaultEmpty}
  NamedRerouteDeclaration(None=@n2) {name=Texture, nodeColor=<rgba 0 0 0 1>}
  NamedRerouteDeclaration(None=@n1.RGB) {name=Thumbnail Color, nodeColor=<rgba 0 0 0 1>}
  NamedRerouteDeclaration(None=MaterialFunctionCall() {materialFunction=&GetUserInterfaceUV}.Normalized UV) {name=Normalized UV, nodeColor=<rgba 0 0 0 1>}
  n3 = MaterialFunctionCall(UV=NamedRerouteUsage(), TextureSize=NamedRerouteUsage()) {materialFunction=&MF_UI_UVScaling}
  NamedRerouteDeclaration(None=@n3.Scaled to Fit) {name=Scale to Fit UV, nodeColor=<rgba 0 0 0 1>}
  NamedRerouteDeclaration(None=@n3.ScaleY) {name=Scale Y UV, nodeColor=<rgba 0 0 0 1>}
  NamedRerouteDeclaration(None=MaterialFunctionCall!(ScreenUV, TextureSize, ScalePivotPoint) {materialFunction=/FortUILibrary/MaterialFunctions/Basics/MF_UI_ScreenSpaceUVScaling}.Scaled to Fit) {name=Screenspace UV, nodeColor=<rgba 0 0 0 1>}
  n4 = NamedRerouteUsage()
  n5 = AppendVector(A=OneMinus(None=ComponentMask(None=@n4) {r=true}), B=ComponentMask(None=@n4) {g=true})
  n6 = MaterialFunctionCall!(
   UV=Multiply(A=StaticSwitchParameter(True=@n5, False=@n4) {desc=Used for localization, group=0 - Global, parameterName=Inverse Order, sortPriority=0}) {constB=1.2, desc=Triangles scale}
   Color A
   Color B
  ) {materialFunction=/FortniteCoreUI/UI/DiscoveryShared/Art/MaterialFunctions/MF_UI_Discovery_TileBacking}
  n7 = MaterialFunctionCall(UV=NamedRerouteUsage()) {materialFunction=&MF_UI_RadialGradient}
  NamedRerouteDeclaration(None=LinearInterpolate(A=NamedRerouteUsage(), B=@n6.Result, Alpha=Multiply(A=@n7.Smooth) {constB=0.5})) {name=Thumbnail Backing, nodeColor=<rgba 0 0 0 1>}
  n8 = ScalarParameter() {group=1 - Thumbnail, parameterName=Thumbnail Zoom, sortPriority=1}
  n9 = Constant() {r=1}
  n10 = Constant()
  n11 = ScalarParameter() {group=1 - Thumbnail, parameterName=Pivot Point Y, sliderMax=0.5, sliderMin=-0.5, sortPriority=2}
  n12 = Subtract(
   A=MaterialFunctionCall(UVs=NamedRerouteUsage(), Texture Scale=Add(A=@n8, B=@n9)) {materialFunction=&ScaleUVsByCenter}.UVs
   B=NamedRerouteDeclaration(None=AppendVector(A=@n10, B=@n11)) {name=Pivot Point, nodeColor=<rgba 0 0 0 1>}
  )
  n13 = NamedRerouteUsage()
  NamedRerouteDeclaration(None=Add(
    A=MaterialFunctionCall(UV=@n12, TextureSize=@n13) {materialFunction=&MF_UI_UVScaling}.Scaled To Fill
    B=NamedRerouteUsage()
   )) {name=Thumbnail UV, nodeColor=<rgba 0 0 0 1>}
  n14 = NamedRerouteUsage()
  n15 = NamedRerouteUsage()
  NamedRerouteDeclaration(None=StaticSwitchParameter(True=LinearInterpolate(A=@n14, Alpha=@n15), False=NamedRerouteUsage()) {defaultValue=true, group=2 - Vertical Vignette, parameterName=Display Vertical Gradient, sortPriority=0}) {name=Thumbnail Vignette, nodeColor=<rgba 0 0 0 1>}
  n16 = ScalarParameter() {defaultValue=0.944, group=3 - Horizontal Vignette, parameterName=Vignette Smooth, sliderMax=1, sortPriority=1}
  n17 = ScalarParameter() {group=3 - Horizontal Vignette, parameterName=Vignette Rotation, sortPriority=0}
  n18 = MaterialFunctionCall(
   UVs=NamedRerouteUsage()
   "Rotation Angle (0-1)"=StaticSwitchParameter(True=OneMinus(None=@n17), False=@n17) {desc=Used for localization, group=0 - Global, parameterName=Inverse Order, sortPriority=0}
  ) {materialFunction=&CustomRotator}
  n19 = StaticSwitchParameter(True=OneMinus(None=@n18.Rotated Values), False=@n18.Rotated Values) {desc=Used for localization, group=0 - Global, parameterName=Inverse Order, sortPriority=0}
  n20 = Add(A=@n19, B=ScalarParameter() {group=3 - Horizontal Vignette, parameterName=Vignette Right Offset, sliderMax=1, sliderMin=-1}) {constB=0.5}
  n21 = Constant()
  n22 = OneMinus(None=Subtract(A=@n19, B=ScalarParameter() {group=3 - Horizontal Vignette, parameterName=Vignette Left Offset, sliderMax=1, sliderMin=-1, sortPriority=31}) {constB=0.5})
  n23 = AppendVector(
   A=StaticSwitchParameter(True=ComponentMask(None=@n20) {r=true}, False=@n21) {defaultValue=true, group=3 - Horizontal Vignette, parameterName=Display Right Gradient, sortPriority=0}
   B=StaticSwitchParameter(True=ComponentMask(None=@n22) {r=true}, False=@n21) {defaultValue=true, group=3 - Horizontal Vignette, parameterName=Display Left Gradient, sortPriority=0}
  )
  n24 = ScalarParameter() {defaultValue=0.944, group=3 - Horizontal Vignette, parameterName=Vignette Dispersion, sliderMax=1, sortPriority=1}
  n25 = Subtract(A=SmoothStep(Min=Multiply(A=@n16) {constB=-1}, Value=Saturate(None=@n23)), B=@n24)
  n26 = Subtract(A=Constant() {desc=Max, r=1}, B=@n24)
  n27 = Saturate(None=Power(
    Base=Divide(A=@n25, B=@n26)
    Exp=ScalarParameter() {defaultValue=1.25, group=3 - Horizontal Vignette, parameterName=Vignette Decay Rate, sortPriority=30}
   ) {constExponent=1.25})
  n28 = ComponentMask(None=@n27) {r=true}
  n29 = Constant()
  n30 = ComponentMask(None=@n27) {g=true}
  NamedRerouteDeclaration(None=LinearInterpolate(
    A=StaticSwitchParameter(True=@n28, False=@n29) {defaultValue=true, group=3 - Horizontal Vignette, parameterName=Display Right Gradient, sortPriority=0}
    Alpha=StaticSwitchParameter(True=@n30, False=@n29) {defaultValue=true, group=3 - Horizontal Vignette, parameterName=Display Left Gradient, sortPriority=0}
   )) {name=Horizontal Gradient Mask, nodeColor=<rgba 0 0 0 1>}
  NamedRerouteDeclaration(None=@n3.ScaleX) {name=Scale X UV, nodeColor=<rgba 0 0 0 1>}
  NamedRerouteDeclaration(None=@n6.Pattern) {name=Triangle Pattern, nodeColor=<rgba 0 0 0 1>}
  n31 = SmoothStep(
   Min=OneMinus(None=ScalarParameter() {defaultValue=0.25, desc=Triangle Distance, group=4 - Triangle Vignette, parameterName=Triangle Distance, sliderMax=1})
   Value=NamedRerouteUsage()
  )
  n32 = NamedRerouteUsage()
  NamedRerouteDeclaration(None=Saturate(None=Multiply(A=@n31, B=@n32))) {name=Triangle Vignette, nodeColor=<rgba 0 0 0 1>}
  NamedRerouteDeclaration(None=@n1.A) {name=Thumbnail Mask, nodeColor=<rgba 0 0 0 1>}
  n33 = Multiply(A=ScalarParameter() {defaultValue=0.944, group=2 - Vertical Vignette, parameterName=Vertical Vignette Smooth, sliderMax=1, sortPriority=2}) {constB=-1}
  n34 = Constant()
  n35 = ScalarParameter() {group=2 - Vertical Vignette, parameterName=Vertical Vignette Offset, sliderMax=1, sliderMin=-1, sortPriority=4}
  n36 = ComponentMask(None=Add(A=NamedRerouteUsage(), B=AppendVector(A=@n34, B=@n35))) {g=true}
  n37 = ScalarParameter() {defaultValue=0.944, group=2 - Vertical Vignette, parameterName=Vertical Vignette Dispersion, sliderMax=1, sortPriority=1}
  n38 = Divide(A=Subtract(A=SmoothStep(Min=@n33, Value=@n36), B=@n37), B=Subtract(A=Constant() {desc=Max, r=1}, B=@n37))
  n39 = ScalarParameter() {defaultValue=1.25, group=2 - Vertical Vignette, parameterName=Vertical Vignette Decay Rate, sortPriority=3}
  NamedRerouteDeclaration(None=Saturate(None=Power(Base=@n38, Exp=@n39) {constExponent=1.25})) {name=Vertical Gradient Mask, nodeColor=<rgba 0 0 0 1>}
  NamedRerouteDeclaration(None=@n3.Scaled To Fill) {name=Scale to Fill UV, nodeColor=<rgba 0 0 0 1>}
  n40 = ScalarParameter() {defaultValue=2048, group=1 - Thumbnail, parameterName=Media Width}
  n41 = ScalarParameter() {defaultValue=1024, group=1 - Thumbnail, parameterName=Media Height, sortPriority=33}
  n42 = NamedRerouteUsage()
  NamedRerouteDeclaration(None=StaticSwitchParameter(
    True=AppendVector(A=@n40, B=@n41) {desc=HACK - Used to send the texture size manually to have both image and video to be the same size}
    False=TextureProperty(Tex=@n42)
   ) {defaultValue=true, group=1 - Thumbnail, parameterName=Use Fixed Size, sortPriority=31}) {name=Texture Size, nodeColor=<rgba 0 0 0 1>}
  n43 = VectorParameter() {group=0 - Global, parameterName=Background Color, sortPriority=1, defaultValue=<rgba 0.0037 0.0137 0.117 1>}
  n44 = VectorParameter() {group=0 - Global, parameterName=Accent Color, sortPriority=2, defaultValue=<rgba 0.007 0.2086 0.8796 1>}
  n45 = NamedRerouteUsage()
  n46 = Constant() {r=-0.2}
  NamedRerouteDeclaration(None=LinearInterpolate(
    A=@n43.RGB
    B=@n44.RGB
    Alpha=MaterialFunctionCall!(UV=@n45, Noise Texture, Rotation Speed=@n46, Noise Scale, Speed A, Speed B, Noise Intensity, Noise Power) {materialFunction=/FortniteCoreUI/UI/DiscoveryShared/Art/MaterialFunctions/MF_UI_Discovery_Clouds}.Result
   )) {name=Outside Background, nodeColor=<rgba 0 0 0 1>}
  n47 = LinearInterpolate(
   A=NamedRerouteUsage()
   B=NamedRerouteUsage()
   Alpha=ScalarParameter() {defaultValue=1, group=1 - Thumbnail, parameterName=Thumbnail Opacity, sliderMax=1, sortPriority=1}
  )
  n48 = NamedRerouteUsage()
  n49 = StaticSwitchParameter(True=LinearInterpolate(A=NamedRerouteUsage(), Alpha=NamedRerouteUsage()), False=NamedRerouteUsage()) {defaultValue=true, group=4 - Triangle Vignette, parameterName=Display Triangle Vignette, sortPriority=0}
  StaticSwitchParameter(
   True=LinearInterpolate(A=NamedRerouteUsage(), B=LinearInterpolate(A=@n47, B=@n48, Alpha=@n49), Alpha=NamedRerouteUsage())
   False=NamedRerouteUsage()
  ) {defaultValue=true, group=1 - Thumbnail, parameterName=Display Thumbnail, sortPriority=0}
 ]
}
