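# UI material for Discovery tile thumbnails, drawing an image over a procedural tile backing.
# Thumbnail UVs are zoomed around a pivot and scaled to fill, with a switchable manual media size hack to match image and video sizing.
# The darkenable thumbnail blends by alpha over the backing, which uses screen-space UV scaling.
# Static switches select thumbnail display, horizontal inversion, and an angled quad mask or bottom fade for opacity.
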Material /FortniteCoreUI/UI/DiscoveryShared/Art/M_UI_Discovery_Thumbnail
graph {
 properties {
  bUsedWithStaticMesh=false
  blendMode=BLEND_Translucent
  materialDomain=MD_UI
  pixelDepthOffsetMode=PDOM_Legacy
  usedShadingModels=Default Lit
 }
 nodes [
  n1 = TextureObjectParameter() {group=1 - Thumbnail, parameterName=Thumbnail, sortPriority=0, texture=/Game/UI/DiscoveryShared/Art/T_UI_Discovery_DefaultEmpty}
  n2 = TextureSample(UVs=NamedRerouteUsage(), Tex=NamedRerouteDeclaration(None=@n1) {name=Thumbnail, nodeColor=<rgba 0 0 0 1>})
  n3 = ScalarParameter() {group=1 - Thumbnail, parameterName=Overlay Opacity, sliderMax=1, sortPriority=3}
  n4 = Constant() {desc=Darken Maximum Amount, r=0.8}
  NamedRerouteDeclaration(None=LinearInterpolate(A=@n2.RGB, B=Constant3Vector() {desc=Darken Color, constant=<rgba 0 0 0 1>}, Alpha=Multiply(A=@n3, B=@n4))) {name=Thumbnail Color, nodeColor=<rgba 0 0 0 1>}
  NamedRerouteDeclaration(None=MaterialFunctionCall() {materialFunction=&GetUserInterfaceUV}.Normalized UV) {name=Normalized UV, nodeColor=<rgba 0 0 0 1>}
  n5 = Constant()
  n6 = ScalarParameter() {group=1 - Thumbnail, parameterName=Pivot Point Y, sliderMax=0.5, sliderMin=-0.5, sortPriority=2}
  n7 = Subtract(
   A=MaterialFunctionCall(UVs=NamedRerouteUsage(), Texture Scale=ScalarParameter() {defaultValue=1, group=1 - Thumbnail, parameterName=Thumbnail Zoom, sortPriority=2}) {materialFunction=&ScaleUVsByCenter}.UVs
   B=NamedRerouteDeclaration(None=AppendVector(A=@n5, B=@n6)) {name=Pivot Point, nodeColor=<rgba 0 0 0 1>}
  )
  n8 = StaticSwitchParameter(
   True=AppendVector(
    A=ScalarParameter() {defaultValue=16, group=1 - Thumbnail, parameterName=Media Width}
    B=ScalarParameter() {defaultValue=9, group=1 - Thumbnail, parameterName=Media Height, sortPriority=33}
   ) {desc=HACK - Used to send the texture size manually to have both image and video to be the same size}
   False=TextureProperty(Tex=NamedRerouteUsage())
  ) {group=1 - Thumbnail, parameterName=Use Fixed Size, sortPriority=31}
  NamedRerouteDeclaration(None=Add(
    A=MaterialFunctionCall(UV=@n7, TextureSize=@n8) {materialFunction=&MF_UI_UVScaling}.Scaled To Fill
    B=NamedRerouteUsage()
   )) {name=Thumbnail UV, nodeColor=<rgba 0 0 0 1>}
  NamedRerouteDeclaration(None=@n2.A) {name=Thumbnail Alpha, nodeColor=<rgba 0 0 0 1>}
  n9 = MaterialFunctionCall!(ScreenUV, TextureSize, ScalePivotPoint) {materialFunction=/FortUILibrary/MaterialFunctions/Basics/MF_UI_ScreenSpaceUVScaling}
  n10 = AppendVector(A=OneMinus(None=ComponentMask(None=@n9.Scaled to Fit) {r=true}), B=ComponentMask(None=@n9.Scaled to Fit) {g=true})
  NamedRerouteDeclaration(None=MaterialFunctionCall!(
    UV=StaticSwitchParameter(True=@n10, False=@n9.Scaled to Fit) {group=0 - Global, parameterName=Inverse Order, sortPriority=0}
    Color A
    Color B
   ) {materialFunction=/FortniteCoreUI/UI/DiscoveryShared/Art/MaterialFunctions/MF_UI_Discovery_TileBacking}.Result) {name=Backing Visual, nodeColor=<rgba 0 0 0 1>}
  n11 = NamedRerouteUsage()
  n12 = Constant() {r=1}
  n13 = MaterialFunctionCall() {materialFunction=&GetUserInterfaceUV}
  n14 = MaterialFunctionCall(Input=OneMinus(None=ComponentMask(None=@n13.Normalized UV) {g=true}), Input Low=Constant() {r=0.1}, Input High=Constant() {r=0.4}) {materialFunction=&RemapValueRange}
  n15 = Constant() {r=1}
  NamedRerouteDeclaration(None=StaticSwitchParameter(
    True=StaticSwitchParameter(True=@n11, False=@n12) {defaultValue=true, group=Quad Mask, parameterName=Use Quad Mask, sortPriority=2}
    False=StaticSwitchParameter(True=@n14.Result, False=@n15) {parameterName=Add Bottom Fade}
   ) {group=0 - Global, parameterName=Use Mask, sortPriority=1}) {name=Opacity, nodeColor=<rgba 0 0 0 1>}
  n16 = NamedRerouteUsage()
  n17 = AppendVector(A=OneMinus(None=ComponentMask(None=@n16) {r=true}), B=ComponentMask(None=@n16) {g=true})
  NamedRerouteDeclaration(None=MaterialFunctionCall!(
    UV=StaticSwitchParameter(True=@n17, False=@n16) {group=0 - Global, parameterName=Inverse Order, sortPriority=0}
    Corner=Constant() {r=1}
    Angle=ScalarParameter() {defaultValue=15, group=Quad Mask, parameterName=Box Angle, sliderMax=90}
    Flip
    Falloff
   ) {materialFunction=/FortUILibrary/MaterialFunctions/Shapes/MF_UI_BoxAngle}.Result) {name=Quad Mask, nodeColor=<rgba 0 0 0 1>}
  NamedRerouteUsage()
  n18 = NamedRerouteUsage()
  n19 = NamedRerouteUsage()
  n20 = ScalarParameter() {group=1 - Thumbnail, parameterName=Thumbnail Alpha, sliderMax=1, sortPriority=1}
  StaticSwitchParameter(True=LinearInterpolate(A=@n18, B=NamedRerouteUsage(), Alpha=Multiply(A=@n19, B=@n20)), False=@n18) {defaultValue=true, group=1 - Thumbnail, parameterName=Display Thumbnail, sortPriority=0}
 ]
}
