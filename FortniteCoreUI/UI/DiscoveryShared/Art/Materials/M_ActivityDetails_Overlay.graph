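# UI material for the Discovery activity details overlay, using Modulate blending to darken what is behind it with a linear gradient.
# Normalized UI UVs pass through CustomRotator by the Rotation parameter, and the green channel blends from the Darkness value to white.
# Vertex alpha then fades that gradient from white, so zero alpha leaves the background untouched under the multiply blend.
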Material /FortniteCoreUI/UI/DiscoveryShared/Art/Materials/M_ActivityDetails_Overlay
graph {
 properties {bUsedWithStaticMesh=false, blendMode=BLEND_Modulate, materialDomain=MD_UI, pixelDepthOffsetMode=PDOM_Legacy, usedShadingModels=Default Lit}
 nodes [
  n1 = MaterialFunctionCall(
   UVs=MaterialFunctionCall() {materialFunction=&GetUserInterfaceUV}.Normalized UV
   "Rotation Angle (0-1)"=ScalarParameter() {parameterName=Rotation, sliderMax=1}
  ) {materialFunction=&CustomRotator}
  LinearInterpolate(
   B=LinearInterpolate(A=ScalarParameter() {defaultValue=0.25, parameterName=Darkness}, Alpha=ComponentMask(None=@n1.Rotated Values) {g=true})
   Alpha=VertexColor().A
  ) {constA=1}
 ]
}
